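# Physics asset for the WonderClef male medium player cosmetic body, covering its dangling belt chain and ring bones on the right side.
# A capsule on dyn_chainB_01 and a box on dyn_chain_ring_f_03 are each constrained to a small-sphere master bone, with limited swing and nearly locked twist.
# Small spheres also sit on the right belt and first ring bones, and a right-thigh capsule provides a leg collision shape.
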PhysicsAsset /BRCosmetics/Characters/Player/Male/Medium/Bodies/M_MED_WonderClef/Meshes/M_MED_WonderClef_Physics
bodies {
 dyn_belt_r [{shapeType=Sphere, radius=1, shapeName=dyn_chain_ring_f_MSTR_sphere}]
 dyn_chainB_01 [{
   shapeType=Capsule
   length=20
   radius=2
   shapeName=dyn_chainB_01_capsule
   center=<16.6873 0 -0.2373>
   rotation=<rot 75.0001 -59.9986 -179.9998>
  }]
 dyn_chainB_MSTR [{shapeType=Sphere, radius=1, shapeName=dyn_chainB_MSTR_sphere}]
 dyn_chain_ring_f_01 [{shapeType=Sphere, radius=1, shapeName=dyn_chain_ring_f_MSTR_sphere}]
 dyn_chain_ring_f_03 [{
   shapeType=Box
   extentX=6.9
   extentY=2.3
   extentZ=6.3
   shapeName=dyn_chain_ring_f_03_box
   center=<1.7401 1.1345 -0.4215>
   rotation=<rot 0.5 -2.5 11>
  }]
 dyn_chain_ring_rotator_MSTR [{shapeType=Sphere, radius=1, shapeName=dyn_chain_ring_f_MSTR_sphere}]
 thigh_r [{shapeType=Capsule, length=30, radius=8, shapeName=thigh_r_capsule, center=<16 0 0>, rotation=<rot 90 0 0>}]
}
constraints [
 {bone1Name=dyn_chainB_01, bone2Name=dyn_chainB_MSTR, swing1LimitDegrees=30, swing2LimitDegrees=30, twistLimitDegrees=1}
 {bone1Name=dyn_chain_ring_f_03, bone2Name=dyn_chain_ring_rotator_MSTR, swing1LimitDegrees=20, swing2LimitDegrees=20, twistLimitDegrees=1}
]
