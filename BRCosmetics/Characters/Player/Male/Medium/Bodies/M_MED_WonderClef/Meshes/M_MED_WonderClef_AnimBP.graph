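# Animation Blueprint for the WonderClef male cosmetic body, adding secondary motion to its hip chain, belts and sleeves.
# It copies the parent mesh pose, runs a Control Rig, then applies Look At and additive corrections driven by movement flags and thigh-to-pelvis angles.
# The cached result feeds RigidBody simulation with a separate skydiving setup, AnimDynamics, or a static chain offset, selected by physics flags.
# Several bone-control nodes are left unconnected.
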AnimBlueprint /BRCosmetics/Characters/Player/Male/Medium/Bodies/M_MED_WonderClef/Meshes/M_MED_WonderClef_AnimBP
target_skeleton=&/M_MED_WonderClef_Skeleton
graphs [
 AnimGraph {nodes [
   "Constraint - dyn_chainB_MSTR (P)"(Alpha=1, ConstraintWeights_0=1) {
    alphaBoolBlend {blendOption=Linear}
    boneToModify {boneName=dyn_chainB_MSTR}
    constraintSetup [{targetBone {boneName=thigh_r}}]
    constraintWeights [1]
   }
   Look At - Bone: dyn_chain_ring_f_01!(ComponentPose, bAlphaBoolEnabled:bool=Get bIsCrouching()) -> (Pose) {
    alphaInputType=Bool
    alphaBoolBlend {blendInTime=0.1, blendOption=Linear, blendOutTime=0.1}
    boneToModify {boneName=dyn_chain_ring_f_01}
    lookAtLocation=<30.2055 -10 8.6384>
    lookAtTarget {boneReference {boneName=thigh_r}}
    lookAt_Axis {axis=<1 0>}
   }
   Look At - Bone: dyn_chain_ring_f_01(bAlphaBoolEnabled=True) {
    alphaInputType=Bool
    alphaBoolBlend {blendInTime=0.1, blendOption=Linear, blendOutTime=0.1}
    boneToModify {boneName=dyn_chain_ring_f_01}
    lookAtLocation=<28.6098 -13.6255>
    lookAtTarget {boneReference {boneName=thigh_r}}
    lookAt_Axis {axis=<1 0>}
   }
   Look At - Bone: dyn_chain_ring_rotator(Alpha=1) {
    alphaBoolBlend {blendOption=Linear}
    boneToModify {boneName=dyn_chain_ring_rotator}
    lookAtLocation=<17.1657 -0.3519 -3.4116>
    lookAtTarget {boneReference {boneName=dyn_chainB_02}}
    lookAt_Axis {axis=<1 0>}
   }
   "Transform (Modify) Bone - Bone: dyn_chain_ring_f_02"!(ComponentPose, bAlphaBoolEnabled:bool=True, Translation="0,0,0", Rotation="0,-15,0", Scale="1,1,1") -> (Pose) {
    alphaInputType=Bool
    rotationMode=BMM_Additive
    rotationSpace=BCS_BoneSpace
    alphaBoolBlend {blendInTime=0.2, blendOption=Linear, blendOutTime=0.4}
    boneToModify {boneName=dyn_chain_ring_f_02}
    rotation {yaw=-15}
   }
   "Transform (Modify) Bone - Bone: dyn_chain_ring_f_02"!(
    ComponentPose
    Alpha:float=Get Right_leg_belt_crouch() -> (Right_leg_belt_crouch:double)
    Translation="0,0,0"
    Rotation="10,-25,0"
    Scale="1,1,1"
   ) -> (Pose) {
    rotationMode=BMM_Additive
    rotationSpace=BCS_BoneSpace
    translationMode=BMM_Additive
    translationSpace=BCS_BoneSpace
    alphaBoolBlend {blendOption=Linear}
    boneToModify {boneName=dyn_chain_ring_f_02}
    rotation {pitch=10, yaw=-25}
   }
   "Transform (Modify) Bone - Bone: dyn_chain_ring_rotator"!(ComponentPose, bAlphaBoolEnabled:bool=Get bIsCrouching(), Translation="0,0,0", Rotation="0,0,0", Scale="1,1,1") -> (Pose) {
    alphaInputType=Bool
    rotationMode=BMM_Additive
    rotationSpace=BCS_BoneSpace
    translationMode=BMM_Additive
    translationSpace=BCS_BoneSpace
    alphaBoolBlend {blendInTime=0.2, blendOption=Linear, blendOutTime=0.4}
    boneToModify {boneName=dyn_chain_ring_rotator}
   }
   n1 = "Transform (Modify) Bone - Bone: dyn_chain_ring_rotator"(bAlphaBoolEnabled=Get bIsCrouching(), Translation="0,0,0", Rotation="-40,-10,0", Scale="1,1,1") {
    alphaInputType=Bool
    rotationMode=BMM_Additive
    rotationSpace=BCS_BoneSpace
    translationMode=BMM_Additive
    translationSpace=BCS_BoneSpace
    alphaBoolBlend {blendInTime=0.2, blendOption=Linear, blendOutTime=0.4}
    boneToModify {boneName=dyn_chain_ring_rotator}
    rotation {pitch=-40, yaw=-10}
   }
   n2 = "Transform (Modify) Bone - Bone: dyn_chain_ring_f_03"!(ComponentPose=@n1, bAlphaBoolEnabled:bool=Get Not crouching() -> (Not crouching:bool), Translation="0,3,0", Rotation="0,0,0", Scale="1,1,1") -> (Pose) {
    alphaInputType=Bool
    rotationMode=BMM_Additive
    rotationSpace=BCS_BoneSpace
    translationMode=BMM_Additive
    translationSpace=BCS_BoneSpace
    alphaBoolBlend {blendOption=Linear}
    boneToModify {boneName=dyn_chain_ring_f_03}
    translation {y=3}
   }
   "Transform (Modify) Bone - Bone: dyn_chain_ring_f_03"!(ComponentPose=@n2, Alpha:float=Get Right_leg_belt_crouch(), Translation="0,4,0", Rotation="0,0,0", Scale="1,1,1") -> (Pose) {
    rotationMode=BMM_Additive
    rotationSpace=BCS_BoneSpace
    translationMode=BMM_Additive
    translationSpace=BCS_BoneSpace
    alphaBoolBlend {blendOption=Linear}
    boneToModify {boneName=dyn_chain_ring_f_03}
    translation {y=4}
   }
   "Transform (Modify) Bone - Bone: dyn_chain_ring_f_03"!(ComponentPose, bAlphaBoolEnabled:bool=Get bIsFalling(), Translation="0,4,0", Rotation="0,0,0", Scale="1,1,1") -> (Pose) {
    alphaInputType=Bool
    rotationMode=BMM_Additive
    rotationSpace=BCS_BoneSpace
    translationMode=BMM_Additive
    translationSpace=BCS_BoneSpace
    alphaBoolBlend {blendInTime=0.5, blendOption=Linear, blendOutTime=0.8}
    boneToModify {boneName=dyn_chain_ring_f_03}
    translation {y=4}
   }
   "Transform (Modify) Bone - Bone: dyn_chainB_02"(Alpha=1, Translation="0,4,-2", Rotation="0,0,0", Scale="1,1,1") {
    rotationMode=BMM_Additive
    rotationSpace=BCS_BoneSpace
    translationMode=BMM_Additive
    translationSpace=BCS_BoneSpace
    alphaBoolBlend {blendInTime=0.5, blendOption=Linear, blendOutTime=0.8}
    boneToModify {boneName=dyn_chainB_02}
    translation {y=4, z=-2}
   }
   "Transform (Modify) Bone - Bone: dyn_chain_ring_f_02"!(ComponentPose, bAlphaBoolEnabled:bool=Get bIsSkyDiving(), Translation="0,0,0", Rotation="0,50,0", Scale="1,1,1") -> (Pose) {
    alphaInputType=Bool
    rotationMode=BMM_Additive
    rotationSpace=BCS_BoneSpace
    translationMode=BMM_Additive
    translationSpace=BCS_BoneSpace
    alphaBoolBlend {blendInTime=0.5, blendOption=Linear, blendOutTime=0.8}
    boneToModify {boneName=dyn_chain_ring_f_02}
    rotation {yaw=50}
   }
   "Transform (Modify) Bone - Bone: dyn_chain_ring_f_02"!(ComponentPose, Alpha:float=1, Translation="0,0,0", Rotation="0,0,0", Scale="1,1,1") -> (Pose) {
    rotationMode=BMM_Additive
    rotationSpace=BCS_BoneSpace
    translationMode=BMM_Additive
    translationSpace=BCS_BoneSpace
    alphaBoolBlend {blendInTime=0.5, blendOption=Linear, blendOutTime=0.8}
    boneToModify {boneName=dyn_chain_ring_f_02}
   }
   "Transform (Modify) Bone - Bone: dyn_chain_ring_rotator"!(ComponentPose, Alpha:float=Get Right_leg_belt_crouch(), Translation="0,0,0", Rotation="0,0,0", Scale="1,1,1") -> (Pose) {
    rotationMode=BMM_Additive
    rotationSpace=BCS_BoneSpace
    translationMode=BMM_Additive
    translationSpace=BCS_BoneSpace
    alphaBoolBlend {blendOption=Linear}
    boneToModify {boneName=dyn_chain_ring_rotator}
   }
   "Transform (Modify) Bone - Bone: dyn_chain_ring_f_02"(Alpha=Get Right_leg_belt_crouch(), Translation="0,0,0", Rotation="0,-50,0", Scale="1,1,1") {
    rotationMode=BMM_Additive
    rotationSpace=BCS_BoneSpace
    translationMode=BMM_Additive
    translationSpace=BCS_BoneSpace
    alphaBoolBlend {blendOption=Linear}
    boneToModify {boneName=dyn_chain_ring_f_02}
    rotation {yaw=-50}
   }
   "Transform (Modify) Bone - Bone: dyn_chain_ring_rotator"(Alpha=Get Right_leg_belt_crouch(), Translation="0,0,0", Rotation="-40,0,0", Scale="1,1,1") {
    rotationMode=BMM_Additive
    rotationSpace=BCS_BoneSpace
    translationMode=BMM_Additive
    translationSpace=BCS_BoneSpace
    alphaBoolBlend {blendInTime=0.2, blendOption=Linear, blendOutTime=0.4}
    boneToModify {boneName=dyn_chain_ring_rotator}
    rotation {pitch=-40}
   }
   "Transform (Modify) Bone - Bone: dyn_chain_ring_f_02"!(ComponentPose, bAlphaBoolEnabled:bool=Get bIsCrouching(), Translation="0,0,0", Rotation="0,0,0", Scale="1,1,1") -> (Pose) {
    alphaInputType=Bool
    rotationMode=BMM_Additive
    rotationSpace=BCS_BoneSpace
    translationMode=BMM_Additive
    translationSpace=BCS_BoneSpace
    alphaBoolBlend {blendInTime=0.2, blendOption=Linear, blendOutTime=0.4}
    boneToModify {boneName=dyn_chain_ring_f_02}
   }
   "Transform (Modify) Bone - Bone: dyn_chain_ring_f_02"!(ComponentPose, Alpha:float=Get Right_leg_belt_crouch(), Translation="0,0,0", Rotation="0,50,0", Scale="1,1,1") -> (Pose) {
    rotationMode=BMM_Additive
    rotationSpace=BCS_BoneSpace
    translationMode=BMM_Additive
    translationSpace=BCS_BoneSpace
    alphaBoolBlend {blendOption=Linear}
    boneToModify {boneName=dyn_chain_ring_f_02}
    rotation {yaw=50}
   }
   n3 = "Transform (Modify) Bone - Bone: dyn_chainB_MSTR"!(
    ComponentPose=Local To Component(LocalPose=Use cached pose 'presim'())
    bAlphaBoolEnabled:bool=Get bIsSurfaceSwimming()
    Translation="0,0,0"
    Rotation="0,-20,0"
    Scale="1,1,1"
   ) -> (Pose) {
    alphaInputType=Bool
    rotationMode=BMM_Additive
    rotationSpace=BCS_BoneSpace
    translationMode=BMM_Additive
    translationSpace=BCS_BoneSpace
    alphaBoolBlend {blendInTime=0.1, blendOption=ExpOut, blendOutTime=0.1}
    boneToModify {boneName=dyn_chainB_MSTR}
    rotation {yaw=-20}
   }
   n4 = "Transform (Modify) Bone - Bone: dyn_chain_ring_f_02"!(
    ComponentPose=@n3
    bAlphaBoolEnabled:bool=Get chain_run_offset() -> (chain_run_offset:bool)
    Translation="0,0,0"
    Rotation="-10,10,0"
    Scale="1,1,1"
   ) -> (Pose) {
    alphaInputType=Bool
    rotationMode=BMM_Additive
    rotationSpace=BCS_BoneSpace
    translationMode=BMM_Additive
    translationSpace=BCS_BoneSpace
    alphaBoolBlend {blendInTime=0.5, blendOption=Linear, blendOutTime=0.8}
    boneToModify {boneName=dyn_chain_ring_f_02}
    rotation {pitch=-10, yaw=10}
   }
   n5 = "Transform (Modify) Bone - Bone: dyn_chain_ring_f_02"!(ComponentPose=@n4, Alpha:float=Get Right_leg_belt_crouch(), Translation="0,0,0", Rotation="0,-20,0", Scale="1,1,1") -> (Pose) {
    rotationMode=BMM_Additive
    rotationSpace=BCS_BoneSpace
    translationMode=BMM_Additive
    translationSpace=BCS_BoneSpace
    alphaBoolBlend {blendInTime=0.2, blendOption=Linear, blendOutTime=0.4}
    boneToModify {boneName=dyn_chain_ring_f_02}
    rotation {yaw=-20}
   }
   n6 = RigidBody(ComponentPose=@n5) {
    bOverrideWorldGravity=true
    overridePhysicsAsset=&/M_MED_WonderClef_Physics
    simulationSpace=BaseBoneSpace
    alphaBoolBlend {blendOption=Linear}
    baseBoneRef {boneName=thigh_r}
    componentAppliedLinearAccClamp=<3000 3000 3000>
    componentLinearAccScale=<1.1 0.6 1.1>
    componentLinearVelScale=<1.3 0.7 1.3>
    overrideWorldGravity {z=-980}
   }
   n7 = RigidBody(ComponentPose=Local To Component(LocalPose=Use cached pose 'presim'())) {
    bOverrideWorldGravity=true
    overridePhysicsAsset=&M_MED_WonderClef_Physics
    simulationSpace=BaseBoneSpace
    alphaBoolBlend {blendOption=Linear}
    baseBoneRef {boneName=thigh_r}
    componentAppliedLinearAccClamp=<3000 3000 3000>
    componentLinearAccScale=<0.2 0.2 0.2>
    componentLinearVelScale=<0.3 0.3 0.3>
    overrideWorldGravity {z=-980}
   }
   n8 = "Transform (Modify) Bone - Bone: dyn_chain_ring_rotator"(ComponentPose=@n7, Alpha=1, Translation="0,0,0", Rotation="-35,0,0", Scale="1,1,1") {
    rotationMode=BMM_Additive
    rotationSpace=BCS_BoneSpace
    translationMode=BMM_Additive
    translationSpace=BCS_BoneSpace
    alphaBoolBlend {blendInTime=0.2, blendOption=Linear, blendOutTime=0.4}
    boneToModify {boneName=dyn_chain_ring_rotator}
    rotation {pitch=-35}
   }
   n9 = Anim Dynamics - Bone: dyn_chain_ring_f_02!(ComponentPose=Local To Component(LocalPose=Use cached pose 'presim'()), bAlphaBoolEnabled:bool=Get bIsCrouching(), ExternalForce="0,0,0") -> (Pose) {
    alphaInputType=Bool
    bUseGravityOverride=true
    simulationSpace=RootRelative
    alphaBoolBlend {blendInTime=0.2, blendOption=Linear, blendOutTime=0.4}
    boundBone {boneName=dyn_chain_ring_f_02}
    componentAppliedLinearAccClamp=<3000 3000 3000>
    componentLinearAccScale=<0.2 0.2 0.2>
    componentLinearVelScale=<0.3 0.3 0.3>
    gravityOverride {z=-980}
    physicsBodyDefinitions [{
      boundBone {boneName=dyn_chain_ring_f_02}
      boxExtents=<5 5 5>
      constraintSetup {angularLimitsMax=<0 0 10>, angularLimitsMin=<0 0 -14>, angularTarget=<0 0 0>}
      localJointOffset=<2 0 0>
     }]
   }
   n10 = Anim Dynamics - Bone: dyn_chainB_01!(ComponentPose=@n9, bAlphaBoolEnabled:bool=Get Not crouching(), ExternalForce="0,0,0") -> (Pose) {
    alphaInputType=Bool
    angularDampingOverride=1
    bOverrideAngularDamping=true
    bUseGravityOverride=true
    simulationSpace=RootRelative
    alphaBoolBlend {blendInTime=0.1, blendOption=Linear, blendOutTime=0.1}
    boundBone {boneName=dyn_chainB_01}
    componentAppliedLinearAccClamp=<3000 3000 3000>
    componentLinearAccScale=<0.2 0.2 0.2>
    componentLinearVelScale=<0.3 0.3 0.3>
    gravityOverride {z=-980}
    physicsBodyDefinitions [{
      boundBone {boneName=dyn_chainB_01}
      boxExtents=<10 10 10>
      constraintSetup {angularLimitsMax=<0 90 5>, angularLimitsMin=<0 20 1>, angularTarget=<0 0 0>}
      localJointOffset=<1 0 0>
     }]
    relativeSpaceBone {boneName=pelvis}
   }
   n11 = Anim Dynamics - Bone: dyn_chainB_01(ComponentPose=@n10, bAlphaBoolEnabled=Get bIsCrouching(), ExternalForce="0,0,0") {
    alphaInputType=Bool
    angularDampingOverride=1
    bOverrideAngularDamping=true
    bUseGravityOverride=true
    simulationSpace=RootRelative
    alphaBoolBlend {blendInTime=0.1, blendOption=Linear, blendOutTime=0.1}
    boundBone {boneName=dyn_chainB_01}
    componentAppliedLinearAccClamp=<3000 3000 3000>
    componentLinearAccScale=<0.2 0.2 0.2>
    componentLinearVelScale=<0.3 0.3 0.3>
    gravityOverride {z=-980}
    physicsBodyDefinitions [{
      boundBone {boneName=dyn_chainB_01}
      boxExtents=<10 10 10>
      constraintSetup {angularLimitsMax=<5 90 20>, angularLimitsMin=<-5 20 0>, angularTarget=<0 0 0>}
      localJointOffset=<1 0 0>
     }]
    relativeSpaceBone {boneName=pelvis}
   }
   n12 = Anim Dynamics - Bone: dyn_sleev_l!(ComponentPose=@n11, bAlphaBoolEnabled:bool=True, ExternalForce="0,0,0") -> (Pose) {
    alphaInputType=Bool
    angularDampingOverride=1
    bOverrideAngularDamping=true
    bUseGravityOverride=true
    simulationSpace=RootRelative
    alphaBoolBlend {blendInTime=0.2, blendOption=Linear, blendOutTime=0.4}
    boundBone {boneName=dyn_sleev_l}
    componentAppliedLinearAccClamp=<3000 3000 3000>
    componentLinearAccScale=<0.1 0.1 0.1>
    componentLinearVelScale=<0.2 0.2 0.2>
    gravityOverride {z=-980}
    physicsBodyDefinitions [{
      boundBone {boneName=dyn_sleev_l}
      boxExtents=<5 20 20>
      constraintSetup {angularLimitsMax=<0 10 10>, angularLimitsMin=<0 -10 -10>, angularTarget=<0 0 0>}
      localJointOffset=<10 0 0>
     }]
   }
   n13 = Anim Dynamics - Bone: dyn_sleev_r!(ComponentPose=@n12, bAlphaBoolEnabled:bool=True, ExternalForce="0,0,0") -> (Pose) {
    alphaInputType=Bool
    angularDampingOverride=1
    bOverrideAngularDamping=true
    bUseGravityOverride=true
    simulationSpace=RootRelative
    alphaBoolBlend {blendInTime=0.2, blendOption=Linear, blendOutTime=0.4}
    boundBone {boneName=dyn_sleev_r}
    componentAppliedLinearAccClamp=<3000 3000 3000>
    componentLinearAccScale=<0.1 0.1 0.1>
    componentLinearVelScale=<0.2 0.2 0.2>
    gravityOverride {z=-980}
    physicsBodyDefinitions [{
      boundBone {boneName=dyn_sleev_r}
      boxExtents=<5 20 20>
      constraintSetup {angularLimitsMax=<0 10 5>, angularLimitsMin=<0 -10 -10>, angularTarget=<0 0 0>}
      localJointOffset=<-10 0 0>
     }]
   }
   n14 = "Transform (Modify) Bone - Bone: dyn_chainB_MSTR"(
    ComponentPose=Local To Component(LocalPose=Use cached pose 'presim'())
    bAlphaBoolEnabled=Get bIsSurfaceSwimming()
    Translation="0,0,0"
    Rotation="-10,5,0"
    Scale="1,1,1"
   ) {
    alphaInputType=Bool
    rotationMode=BMM_Additive
    rotationSpace=BCS_BoneSpace
    translationMode=BMM_Additive
    translationSpace=BCS_BoneSpace
    alphaBoolBlend {blendInTime=0.1, blendOption=ExpOut, blendOutTime=0.1}
    boneToModify {boneName=dyn_chainB_MSTR}
    rotation {pitch=-10, yaw=5}
   }
   n15 = "Transform (Modify) Bone - Bone: dyn_chainB_MSTR"(ComponentPose=@n14, bAlphaBoolEnabled=True, Translation="0,0,0", Rotation="-20,5,0", Scale="1,1,1") {
    alphaInputType=Bool
    rotationMode=BMM_Additive
    rotationSpace=BCS_BoneSpace
    translationMode=BMM_Additive
    translationSpace=BCS_BoneSpace
    alphaBoolBlend {blendInTime=0.1, blendOption=ExpOut, blendOutTime=0.1}
    boneToModify {boneName=dyn_chainB_MSTR}
    rotation {pitch=-20, yaw=5}
   }
   n16 = Blend Poses by bool(
    BlendPose_0=Blend Poses by bool(BlendPose_0=Component To Local(ComponentPose=@n8), BlendPose_1=Component To Local(ComponentPose=@n6), bActiveValue=Get bIsSkyDiving()) {blendType=Linear}
    BlendPose_1=Blend Poses by bool(
     BlendPose_0=Component To Local(ComponentPose=@n13)
     BlendPose_1=Component To Local(ComponentPose=@n15)
     bActiveValue=Get bAnimDynamicsEnabled() -> (bAnimDynamicsEnabled:bool)
    ) {blendType=Linear}
    bActiveValue=Get bRigidBodyEnabled()
   ) {blendType=Linear}
   Output Pose(Result=Component To Local(ComponentPose=Local To Component(LocalPose=@n16))) {layerGroup=None}
   n17 = Control Rig(Source=Copy Pose From Mesh() {bUseAttachedParent=true}) {
    alphaBoolBlend {blendOption=Linear}
    controlRigAssetReference {blueprintRigClass=&/M_MED_WonderClef_CtrlRig.M_MED_WonderClef_CtrlRig_C}
   }
   n18 = Look At - Bone: dyn_chainB_MSTR(ComponentPose=Local To Component(LocalPose=@n17), Alpha=1) {
    alphaBoolBlend {blendOption=Linear}
    boneToModify {boneName=dyn_chainB_MSTR}
    lookAtLocation=<14.6789 0 8.5514>
    lookAtTarget {boneReference {boneName=thigh_r}}
    lookAt_Axis {axis=<1 0>}
   }
   n19 = Look At - Bone: dyn_chain_ring_f_01(ComponentPose=@n18, bAlphaBoolEnabled=Get Not Jumping() -> (Not Jumping:bool)) {
    alphaInputType=Bool
    alphaBoolBlend {blendInTime=0.2, blendOption=Linear, blendOutTime=0.4}
    boneToModify {boneName=dyn_chain_ring_f_01}
    lookAtLocation=<28.6098 -13.6255>
    lookAtTarget {boneReference {boneName=thigh_r}}
    lookAt_Axis {axis=<1 0>}
   }
   n20 = Look At - Bone: dyn_chain_ring_f_01(ComponentPose=@n19, bAlphaBoolEnabled=Get bIsJumping()) {
    alphaInputType=Bool
    alphaBoolBlend {blendInTime=0.2, blendOption=Linear, blendOutTime=0.4}
    boneToModify {boneName=dyn_chain_ring_f_01}
    lookAtLocation=<28.6098 -22.8717 0>
    lookAtTarget {boneReference {boneName=thigh_r}}
    lookAt_Axis {axis=<1 0>}
   }
   n21 = "Transform (Modify) Bone - Bone: dyn_chain_ring_rotator"!(ComponentPose=@n20, bAlphaBoolEnabled:bool=Get bIsCrouching(), Translation="0,0,0", Rotation="-35,0,0", Scale="1,1,1") -> (Pose) {
    alphaInputType=Bool
    rotationMode=BMM_Additive
    rotationSpace=BCS_BoneSpace
    translationMode=BMM_Additive
    translationSpace=BCS_BoneSpace
    alphaBoolBlend {blendInTime=0.2, blendOption=Linear, blendOutTime=0.4}
    boneToModify {boneName=dyn_chain_ring_rotator}
    rotation {pitch=-35}
   }
   n22 = "Transform (Modify) Bone - Bone: dyn_chain_ring_f_01"!(ComponentPose=@n21, bAlphaBoolEnabled:bool=Get Not crouching(), Translation="0,0,0", Rotation="0,0,0", Scale="1,1,1") -> (Pose) {
    alphaInputType=Bool
    rotationMode=BMM_Additive
    rotationSpace=BCS_BoneSpace
    alphaBoolBlend {blendInTime=0.2, blendOption=Linear, blendOutTime=0.4}
    boneToModify {boneName=dyn_chain_ring_f_01}
   }
   n23 = "Transform (Modify) Bone - Bone: dyn_chain_ring_f_01"(ComponentPose=@n22, bAlphaBoolEnabled=Get bIsCrouching(), Translation="0,0,0", Rotation="0,-17,0", Scale="1,1,1") {
    alphaInputType=Bool
    rotationMode=BMM_Additive
    rotationSpace=BCS_BoneSpace
    alphaBoolBlend {blendInTime=0.2, blendOption=Linear, blendOutTime=0.4}
    boneToModify {boneName=dyn_chain_ring_f_01}
    rotation {yaw=-17}
   }
   n24 = "Transform (Modify) Bone - Bone: dyn_belt_r"(ComponentPose=@n23, Alpha=Get Right_leg_belt_crouch(), Translation="-5,0.5,0", Rotation="0,0,0", Scale="1,1,1") {
    rotationMode=BMM_Additive
    rotationSpace=BCS_BoneSpace
    translationMode=BMM_Additive
    translationSpace=BCS_BoneSpace
    alphaBoolBlend {blendOption=Linear}
    boneToModify {boneName=dyn_belt_r}
    translation=<-5 0.5>
   }
   n25 = "Transform (Modify) Bone - Bone: dyn_belt_l"(ComponentPose=@n24, Alpha=Get Left_leg_belt_crouch() -> (Left_leg_belt_crouch:float), Translation="-5,0,0", Rotation="0,0,0", Scale="1,1,1") {
    rotationMode=BMM_Additive
    rotationSpace=BCS_BoneSpace
    translationMode=BMM_Additive
    translationSpace=BCS_BoneSpace
    alphaBoolBlend {blendOption=Linear}
    boneToModify {boneName=dyn_belt_l}
    translation {x=-5}
   }
   n26 = "Transform (Modify) Bone - Bone: dyn_chainB_MSTR"(ComponentPose=@n25, bAlphaBoolEnabled=Get Idle_chain() -> (Idle_chain:bool), Translation="0,0,0", Rotation="15,0,0", Scale="1,1,1") {
    alphaInputType=Bool
    rotationMode=BMM_Additive
    rotationSpace=BCS_BoneSpace
    translationMode=BMM_Additive
    translationSpace=BCS_BoneSpace
    alphaBoolBlend {blendInTime=0.1, blendOption=ExpOut, blendOutTime=0.1}
    boneToModify {boneName=dyn_chainB_MSTR}
    rotation {pitch=15}
   }
   n27 = "Transform (Modify) Bone - Bone: dyn_chainB_MSTR"(ComponentPose=@n26, bAlphaBoolEnabled=Get bIsSurfaceSwimming(), Translation="0,0,0", Rotation="0,-10,0", Scale="1,1,1") {
    alphaInputType=Bool
    rotationMode=BMM_Additive
    rotationSpace=BCS_BoneSpace
    translationMode=BMM_Additive
    translationSpace=BCS_BoneSpace
    alphaBoolBlend {blendInTime=0.1, blendOption=ExpOut, blendOutTime=0.1}
    boneToModify {boneName=dyn_chainB_MSTR}
    rotation {yaw=-10}
   }
   n28 = "Transform (Modify) Bone - Bone: dyn_chainB_MSTR"(ComponentPose=@n27, bAlphaBoolEnabled=Get bIsJumping(), Translation="0,0,0", Rotation="0,15,0", Scale="1,1,1") {
    alphaInputType=Bool
    rotationMode=BMM_Additive
    rotationSpace=BCS_BoneSpace
    translationMode=BMM_Additive
    translationSpace=BCS_BoneSpace
    alphaBoolBlend {blendInTime=0.1, blendOption=ExpOut, blendOutTime=0.1}
    boneToModify {boneName=dyn_chainB_MSTR}
    rotation {yaw=15}
   }
   n29 = "Transform (Modify) Bone - Bone: dyn_chainB_02"(ComponentPose=@n28, Alpha=Get Right_leg_belt_crouch(), Translation="-5,1,0", Rotation="0,0,0", Scale="1,1,1") {
    rotationMode=BMM_Additive
    rotationSpace=BCS_BoneSpace
    translationMode=BMM_Additive
    translationSpace=BCS_BoneSpace
    alphaBoolBlend {blendInTime=0.5, blendOption=Linear, blendOutTime=0.8}
    boneToModify {boneName=dyn_chainB_02}
    translation=<-5 1>
   }
   Save cached pose 'presim'(Pose=Component To Local(ComponentPose=@n29)) {cacheName=presim}
   Get bIsCrouching()
  ]}
 EventGraph {nodes [
   Event BlueprintUpdateAnimation()
   > Set Right_leg_belt_crouch(Right_leg_belt_crouch:double=GetAlphaFromReferenceTransformDelta(SkeletalMeshComponent=Get Skel Mesh(), BoneName=thigh_r, BaseName=pelvis, InputRange=(X=-18,Y=-98), TwistAxis="1,0,0")) -> (Output_Get:double)
   > Set Left_leg_belt_crouch(Left_leg_belt_crouch:float=GetAlphaFromReferenceTransformDelta(SkeletalMeshComponent=Get Skel Mesh(), BoneName=thigh_l, BaseName=pelvis, InputRange=(X=-18,Y=-98), TwistAxis="1,0,0")) -> (Output_Get:float)
   > Set Not crouching(Not crouching:bool=NOT Boolean(A=Get bIsCrouching())) -> (Output_Get:bool)
   > Set Not Jumping(Not Jumping:bool=NOT Boolean(A=Get bIsJumping())) -> (Output_Get:bool)
   > Set chain_run_offset(chain_run_offset:bool=AND Boolean(A=Get bIsMoving2D(), B=Get Not crouching())) -> (Output_Get:bool)
   n1 = Get bIsMoving2D()
   n2 = Get bIsCrouching()
   n3 = Get bIsSkydiveFloating()
   n4 = Get bIsJumping()
   > Set Idle_chain(Idle_chain:bool=OR Boolean!(A=NOT Boolean(A=@n2), B=NOT Boolean(A=@n1), C:bool=NOT Boolean(A=@n3), D:bool=NOT Boolean(A=@n4)) -> (ReturnValue)) -> (Output_Get:bool)
   n5 = Event OnCharacterCustomizationLoaded()
   > Set Skel Mesh(Skel Mesh=Get Array Item with EFort Custom Part Type(Array=@n5.CustomMeshParts))
  ]}
]
variables [
 chain_run_offset {type=boolean}
 Idle_chain {type=boolean}
 Left_leg_belt_crouch {type=number}
 Not crouching {type=boolean}
 Not Jumping {type=boolean}
 Right_leg_belt_crouch {type=number}
 Skel Mesh {type=SkeletalMeshComponent}
]
