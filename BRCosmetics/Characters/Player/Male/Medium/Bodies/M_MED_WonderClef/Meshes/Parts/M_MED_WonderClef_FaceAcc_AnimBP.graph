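# Post-process style AnimBP for the WonderClef face-accessory mesh, animating headphones, hair and a face accessory on the player body.
# It copies the parent pose, adjusts headphones with constraints, look-at and offsets for crouch, targeting, melee and smoothed head/arm deltas from EventGraph.
# It pins the face accessory to weighted lip, jaw and chin bones, then caches that pose.
# A front-end switch selects an earcup tweak before thirty AnimDynamics hair nodes.
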AnimBlueprint /BRCosmetics/Characters/Player/Male/Medium/Bodies/M_MED_WonderClef/Meshes/Parts/M_MED_WonderClef_FaceAcc_AnimBP
target_skeleton=&/M_MED_WonderClef_FaceAcc_Skeleton
graphs [
 AnimGraph {nodes [
   n1 = Get bIsPlayingMeleeAnim()
   n2 = "Constraint - dyn_headphones (P)"!(ComponentPose, bAlphaBoolEnabled:bool=@n1, ConstraintWeights_0=1) -> (Pose) {
    alphaInputType=Bool
    alphaBoolBlend {blendOption=Linear}
    boneToModify {boneName=dyn_headphones}
    constraintSetup [{targetBone {boneName=head}}]
    constraintWeights [1]
   }
   n3 = "Transform (Modify) Bone - Bone: dyn_headphones"!(ComponentPose=@n2, bAlphaBoolEnabled:bool=@n1, Translation="-6,-7,0", Rotation="0,40,0", Scale="1,1,1") -> (Pose) {
    alphaInputType=Bool
    rotationMode=BMM_Additive
    rotationSpace=BCS_BoneSpace
    translationMode=BMM_Additive
    translationSpace=BCS_BoneSpace
    alphaBoolBlend {blendInTime=0.1, blendOption=ExpIn, blendOutTime=0.1}
    boneToModify {boneName=dyn_headphones}
    rotation {yaw=40}
    translation=<-6 -7>
   }
   n4 = "Transform (Modify) Bone - Bone: dyn_headphones_r"!(ComponentPose=@n3, bAlphaBoolEnabled:bool=@n1, Translation="0,0,-2", Rotation="5,0,10", Scale="1,1,1") -> (Pose) {
    alphaInputType=Bool
    rotationMode=BMM_Additive
    rotationSpace=BCS_BoneSpace
    translationMode=BMM_Additive
    translationSpace=BCS_BoneSpace
    alphaBoolBlend {blendInTime=0.1, blendOption=ExpIn, blendOutTime=0.1}
    boneToModify {boneName=dyn_headphones_r}
    rotation {pitch=5, roll=10}
    translation {z=-2}
   }
   "Transform (Modify) Bone - Bone: dyn_headphones_l"!(ComponentPose=@n4, bAlphaBoolEnabled:bool=@n1, Translation="0,0,2", Rotation="-5,0,-10", Scale="1,1,1") -> (Pose) {
    alphaInputType=Bool
    rotationMode=BMM_Additive
    rotationSpace=BCS_BoneSpace
    translationMode=BMM_Additive
    translationSpace=BCS_BoneSpace
    alphaBoolBlend {blendInTime=0.1, blendOption=ExpIn, blendOutTime=0.1}
    boneToModify {boneName=dyn_headphones_l}
    rotation {pitch=-5, roll=-10}
    translation {z=2}
   }
   n5 = Get bIsPlayingMeleeAnim()
   "Transform (Modify) Bone - Bone: dyn_headphones"(bAlphaBoolEnabled=@n5, Translation="-1,0,0", Rotation="0,0,0", Scale="1,1,1") {
    alphaInputType=Bool
    rotationMode=BMM_Additive
    rotationSpace=BCS_BoneSpace
    translationMode=BMM_Additive
    translationSpace=BCS_BoneSpace
    alphaBoolBlend {blendInTime=0.1, blendOption=ExpIn, blendOutTime=0.1}
    boneToModify {boneName=dyn_headphones}
    translation {x=-1}
   }
   "Transform (Modify) Bone - Bone: dyn_headphones"(bAlphaBoolEnabled=@n5, Translation="0,-1,0", Rotation="0,0,0", Scale="1,1,1") {
    alphaInputType=Bool
    rotationMode=BMM_Additive
    rotationSpace=BCS_BoneSpace
    translationMode=BMM_Additive
    translationSpace=BCS_BoneSpace
    alphaBoolBlend {blendInTime=0.1, blendOption=Linear, blendOutTime=0.3}
    boneToModify {boneName=dyn_headphones}
    translation {y=-1}
   }
   n6 = "Transform (Modify) Bone - Bone: dyn_headphones_r_MSTR"(
    ComponentPose=Local To Component(LocalPose=Use cached pose 'Headphones'())
    Alpha=Get Headphone_Right() -> (Headphone_Right:float)
    Translation="0,0,0"
    Rotation="0,0,-6.5"
    Scale="1,1,1"
   ) {
    rotationMode=BMM_Additive
    rotationSpace=BCS_BoneSpace
    translationMode=BMM_Additive
    translationSpace=BCS_BoneSpace
    alphaBoolBlend {blendInTime=0.2, blendOption=Linear, blendOutTime=0.6}
    boneToModify {boneName=dyn_headphones_r_MSTR}
    rotation {roll=-6.5}
   }
   n7 = Component To Local(ComponentPose=@n6)
   n8 = Use cached pose 'Headphones'()
   n9 = Get bIsInFrontEnd()
   n10 = Anim Dynamics - Bone: dyn_hair_1!(
    ComponentPose=Local To Component(LocalPose=Blend Poses by bool(BlendPose_0=@n7, BlendPose_1=@n8, bActiveValue=@n9) {blendType=Linear})
    bAlphaBoolEnabled:bool=True
    ExternalForce="0,0,0"
   ) -> (Pose) {
    alphaInputType=Bool
    angularDampingOverride=1
    bOverrideAngularDamping=true
    alphaBoolBlend {blendInTime=0.1, blendOption=Linear, blendOutTime=0.2}
    boundBone {boneName=dyn_hair_1}
    componentAppliedLinearAccClamp=<3000 3000 3000>
    componentLinearAccScale=<0.2 0.2 0.2>
    componentLinearVelScale=<0.3 0.3 0.3>
    physicsBodyDefinitions [{
      boundBone {boneName=dyn_hair_1}
      boxExtents=<7 7 7>
      constraintSetup {angularLimitsMax=<0 5 45>, angularLimitsMin=<0 -5 -45>, angularTarget=<0 0 0>}
      localJointOffset=<1 0 0>
     }]
   }
   n11 = Anim Dynamics - Bone: dyn_hair_2!(ComponentPose=@n10, bAlphaBoolEnabled:bool=True, ExternalForce="0,0,0") -> (Pose) {
    alphaInputType=Bool
    angularBiasOverride=0.9
    angularDampingOverride=0.99
    angularSpringConstant=2000
    bAngularSpring=true
    bOverrideAngularDamping=true
    bUseGravityOverride=true
    alphaBoolBlend {blendInTime=0.1, blendOption=Custom, blendOutTime=0.2}
    boundBone {boneName=dyn_hair_2}
    componentAppliedLinearAccClamp=<1000 1000 1000>
    componentLinearAccScale=<0.3 0.3 0.3>
    componentLinearVelScale=<0.5 0.5 0.5>
    gravityOverride {z=-100}
    physicsBodyDefinitions [{
      boundBone {boneName=dyn_hair_2}
      boxExtents=<6 6 8>
      constraintSetup {angularLimitsMax=<5 5 15>, angularLimitsMin=<-5 -5 -15>, angularTarget=<0 0 0>, linearAxesMin=<-0.1 0 0>}
      localJointOffset=<1 0 0>
     }]
   }
   n12 = Anim Dynamics - Bone: dyn_hair_3!(ComponentPose=@n11, bAlphaBoolEnabled:bool=True, ExternalForce="0,0,0") -> (Pose) {
    alphaInputType=Bool
    angularBiasOverride=0.9
    angularDampingOverride=0.99
    angularSpringConstant=2000
    bAngularSpring=true
    bOverrideAngularDamping=true
    bUseGravityOverride=true
    alphaBoolBlend {blendInTime=0.1, blendOption=Custom, blendOutTime=0.2}
    boundBone {boneName=dyn_hair_3}
    componentAppliedLinearAccClamp=<1500 1500 1500>
    componentLinearAccScale=<0.3 0.3 0.3>
    componentLinearVelScale=<0.5 0.5 0.5>
    gravityOverride {z=-100}
    physicsBodyDefinitions [{
      boundBone {boneName=dyn_hair_3}
      boxExtents=<5 5 10>
      constraintSetup {angularLimitsMax=<5 5 15>, angularLimitsMin=<-5 -5 -15>, angularTarget=<0 0 0>, linearAxesMin=<-0.1 0 0>}
      localJointOffset=<1 0 0>
     }]
   }
   n13 = Anim Dynamics - Bone: dyn_hair_4!(ComponentPose=@n12, bAlphaBoolEnabled:bool=True, ExternalForce="0,0,0") -> (Pose) {
    alphaInputType=Bool
    angularBiasOverride=0.9
    angularDampingOverride=0.99
    angularSpringConstant=2000
    bAngularSpring=true
    bOverrideAngularDamping=true
    bUseGravityOverride=true
    alphaBoolBlend {blendInTime=0.1, blendOption=Custom, blendOutTime=0.2}
    boundBone {boneName=dyn_hair_4}
    componentAppliedLinearAccClamp=<2000 2000 2000>
    componentLinearAccScale=<0.3 0.3 0.3>
    componentLinearVelScale=<0.5 0.5 0.5>
    gravityOverride {z=-100}
    physicsBodyDefinitions [{
      boundBone {boneName=dyn_hair_4}
      boxExtents=<6 6 8>
      constraintSetup {angularLimitsMax=<5 5 15>, angularLimitsMin=<-5 -5 -15>, angularTarget=<0 0 0>, linearAxesMin=<-0.1 0 0>}
      localJointOffset=<1 0 0>
     }]
   }
   n14 = Anim Dynamics - Bone: dyn_hair_5!(ComponentPose=@n13, bAlphaBoolEnabled:bool=True, ExternalForce="0,0,0") -> (Pose) {
    alphaInputType=Bool
    angularBiasOverride=0.9
    angularDampingOverride=0.99
    angularSpringConstant=2000
    bAngularSpring=true
    bOverrideAngularDamping=true
    bUseGravityOverride=true
    alphaBoolBlend {blendInTime=0.1, blendOption=Custom, blendOutTime=0.2}
    boundBone {boneName=dyn_hair_5}
    componentAppliedLinearAccClamp=<2500 2500 2500>
    componentLinearAccScale=<0.3 0.3 0.3>
    componentLinearVelScale=<0.5 0.5 0.5>
    gravityOverride {z=-100}
    physicsBodyDefinitions [{
      boundBone {boneName=dyn_hair_5}
      boxExtents=<5 5 10>
      constraintSetup {angularLimitsMax=<5 5 15>, angularLimitsMin=<-5 -5 -15>, angularTarget=<0 0 0>, linearAxesMin=<-0.1 0 0>}
      localJointOffset=<1 0 0>
     }]
   }
   n15 = Anim Dynamics - Bone: dyn_hair_6!(ComponentPose=@n14, bAlphaBoolEnabled:bool=True, ExternalForce="0,0,0") -> (Pose) {
    alphaInputType=Bool
    angularBiasOverride=0.9
    angularDampingOverride=0.99
    angularSpringConstant=2000
    bAngularSpring=true
    bOverrideAngularDamping=true
    bUseGravityOverride=true
    alphaBoolBlend {blendInTime=0.1, blendOption=Custom, blendOutTime=0.2}
    boundBone {boneName=dyn_hair_6}
    componentAppliedLinearAccClamp=<500 500 500>
    componentLinearAccScale=<0.3 0.3 0.3>
    componentLinearVelScale=<0.5 0.5 0.5>
    gravityOverride {z=-100}
    physicsBodyDefinitions [{
      boundBone {boneName=dyn_hair_6}
      boxExtents=<5 5 10>
      constraintSetup {angularLimitsMax=<5 5 15>, angularLimitsMin=<-5 -5 -15>, angularTarget=<0 0 0>, linearAxesMin=<-0.1 0 0>}
      localJointOffset=<1 0 0>
     }]
   }
   n16 = Anim Dynamics - Bone: dyn_hair_7!(ComponentPose=@n15, bAlphaBoolEnabled:bool=True, ExternalForce="0,0,0") -> (Pose) {
    alphaInputType=Bool
    angularBiasOverride=0.9
    angularDampingOverride=0.99
    angularSpringConstant=2000
    bAngularSpring=true
    bOverrideAngularDamping=true
    bUseGravityOverride=true
    alphaBoolBlend {blendInTime=0.1, blendOption=Custom, blendOutTime=0.2}
    boundBone {boneName=dyn_hair_7}
    componentAppliedLinearAccClamp=<1000 1000 1000>
    componentLinearAccScale=<0.3 0.3 0.3>
    componentLinearVelScale=<0.5 0.5 0.5>
    gravityOverride {z=-100}
    physicsBodyDefinitions [{
      boundBone {boneName=dyn_hair_7}
      boxExtents=<6 6 8>
      constraintSetup {angularLimitsMax=<5 5 15>, angularLimitsMin=<-5 -5 -15>, angularTarget=<0 0 0>, linearAxesMin=<-0.1 0 0>}
      localJointOffset=<1 0 0>
     }]
   }
   n17 = Anim Dynamics - Bone: dyn_hair_8!(ComponentPose=@n16, bAlphaBoolEnabled:bool=True, ExternalForce="0,0,0") -> (Pose) {
    alphaInputType=Bool
    angularBiasOverride=0.9
    angularDampingOverride=0.99
    angularSpringConstant=2000
    bAngularSpring=true
    bOverrideAngularDamping=true
    bUseGravityOverride=true
    alphaBoolBlend {blendInTime=0.1, blendOption=Custom, blendOutTime=0.2}
    boundBone {boneName=dyn_hair_8}
    componentAppliedLinearAccClamp=<1500 1500 1500>
    componentLinearAccScale=<0.3 0.3 0.3>
    componentLinearVelScale=<0.5 0.5 0.5>
    gravityOverride {z=-100}
    physicsBodyDefinitions [{
      boundBone {boneName=dyn_hair_8}
      boxExtents=<5 5 10>
      constraintSetup {angularLimitsMax=<5 5 15>, angularLimitsMin=<-5 -5 -15>, angularTarget=<0 0 0>, linearAxesMin=<-0.1 0 0>}
      localJointOffset=<1 0 0>
     }]
   }
   n18 = Anim Dynamics - Bone: dyn_hair_9!(ComponentPose=@n17, bAlphaBoolEnabled:bool=True, ExternalForce="0,0,0") -> (Pose) {
    alphaInputType=Bool
    angularBiasOverride=0.9
    angularDampingOverride=0.99
    angularSpringConstant=2000
    bAngularSpring=true
    bOverrideAngularDamping=true
    bUseGravityOverride=true
    alphaBoolBlend {blendInTime=0.1, blendOption=Custom, blendOutTime=0.2}
    boundBone {boneName=dyn_hair_9}
    componentAppliedLinearAccClamp=<2000 2000 2000>
    componentLinearAccScale=<0.3 0.3 0.3>
    componentLinearVelScale=<0.5 0.5 0.5>
    gravityOverride {z=-100}
    physicsBodyDefinitions [{
      boundBone {boneName=dyn_hair_9}
      boxExtents=<6 6 8>
      constraintSetup {angularLimitsMax=<5 5 15>, angularLimitsMin=<-5 -5 -15>, angularTarget=<0 0 0>, linearAxesMin=<-0.1 0 0>}
      localJointOffset=<1 0 0>
     }]
   }
   n19 = Anim Dynamics - Bone: dyn_hair_10!(ComponentPose=@n18, bAlphaBoolEnabled:bool=True, ExternalForce="0,0,0") -> (Pose) {
    alphaInputType=Bool
    angularBiasOverride=0.9
    angularDampingOverride=0.99
    angularSpringConstant=2000
    bAngularSpring=true
    bOverrideAngularDamping=true
    bUseGravityOverride=true
    alphaBoolBlend {blendInTime=0.1, blendOption=Custom, blendOutTime=0.2}
    boundBone {boneName=dyn_hair_10}
    componentAppliedLinearAccClamp=<2500 2500 2500>
    componentLinearAccScale=<0.3 0.3 0.3>
    componentLinearVelScale=<0.5 0.5 0.5>
    gravityOverride {z=-100}
    physicsBodyDefinitions [{
      boundBone {boneName=dyn_hair_10}
      boxExtents=<5 5 10>
      constraintSetup {angularLimitsMax=<5 5 15>, angularLimitsMin=<-5 -5 -15>, angularTarget=<0 0 0>, linearAxesMin=<-0.1 0 0>}
      localJointOffset=<1 0 0>
     }]
   }
   n20 = Anim Dynamics - Bone: dyn_hair_11!(ComponentPose=@n19, bAlphaBoolEnabled:bool=True, ExternalForce="0,0,0") -> (Pose) {
    alphaInputType=Bool
    angularBiasOverride=0.9
    angularDampingOverride=0.99
    angularSpringConstant=2000
    bAngularSpring=true
    bOverrideAngularDamping=true
    bUseGravityOverride=true
    alphaBoolBlend {blendInTime=0.1, blendOption=Custom, blendOutTime=0.2}
    boundBone {boneName=dyn_hair_11}
    componentAppliedLinearAccClamp=<500 500 500>
    componentLinearAccScale=<0.3 0.3 0.3>
    componentLinearVelScale=<0.5 0.5 0.5>
    gravityOverride {z=-100}
    physicsBodyDefinitions [{
      boundBone {boneName=dyn_hair_11}
      boxExtents=<5 5 10>
      constraintSetup {angularLimitsMax=<5 5 15>, angularLimitsMin=<-5 -5 -15>, angularTarget=<0 0 0>, linearAxesMin=<-0.1 0 0>}
      localJointOffset=<1 0 0>
     }]
   }
   n21 = Anim Dynamics - Bone: dyn_hair_12!(ComponentPose=@n20, bAlphaBoolEnabled:bool=True, ExternalForce="0,0,0") -> (Pose) {
    alphaInputType=Bool
    angularBiasOverride=0.9
    angularDampingOverride=0.99
    angularSpringConstant=2000
    bAngularSpring=true
    bOverrideAngularDamping=true
    bUseGravityOverride=true
    alphaBoolBlend {blendInTime=0.1, blendOption=Custom, blendOutTime=0.2}
    boundBone {boneName=dyn_hair_12}
    componentAppliedLinearAccClamp=<1000 1000 1000>
    componentLinearAccScale=<0.3 0.3 0.3>
    componentLinearVelScale=<0.5 0.5 0.5>
    gravityOverride {z=-100}
    physicsBodyDefinitions [{
      boundBone {boneName=dyn_hair_12}
      boxExtents=<6 6 8>
      constraintSetup {angularLimitsMax=<5 5 15>, angularLimitsMin=<-5 -5 -15>, angularTarget=<0 0 0>, linearAxesMin=<-0.1 0 0>}
      localJointOffset=<1 0 0>
     }]
   }
   n22 = Anim Dynamics - Bone: dyn_hair_13!(ComponentPose=@n21, bAlphaBoolEnabled:bool=True, ExternalForce="0,0,0") -> (Pose) {
    alphaInputType=Bool
    angularBiasOverride=0.9
    angularDampingOverride=0.99
    angularSpringConstant=2000
    bAngularSpring=true
    bOverrideAngularDamping=true
    bUseGravityOverride=true
    alphaBoolBlend {blendInTime=0.1, blendOption=Custom, blendOutTime=0.2}
    boundBone {boneName=dyn_hair_13}
    componentAppliedLinearAccClamp=<1500 1500 1500>
    componentLinearAccScale=<0.3 0.3 0.3>
    componentLinearVelScale=<0.5 0.5 0.5>
    gravityOverride {z=-100}
    physicsBodyDefinitions [{
      boundBone {boneName=dyn_hair_13}
      boxExtents=<5 5 10>
      constraintSetup {angularLimitsMax=<5 5 15>, angularLimitsMin=<-5 -5 -15>, angularTarget=<0 0 0>, linearAxesMin=<-0.1 0 0>}
      localJointOffset=<1 0 0>
     }]
   }
   n23 = Anim Dynamics - Bone: dyn_hair_14!(ComponentPose=@n22, bAlphaBoolEnabled:bool=True, ExternalForce="0,0,0") -> (Pose) {
    alphaInputType=Bool
    angularBiasOverride=0.9
    angularDampingOverride=0.99
    angularSpringConstant=2000
    bAngularSpring=true
    bOverrideAngularDamping=true
    bUseGravityOverride=true
    alphaBoolBlend {blendInTime=0.1, blendOption=Custom, blendOutTime=0.2}
    boundBone {boneName=dyn_hair_14}
    componentAppliedLinearAccClamp=<2000 2000 2000>
    componentLinearAccScale=<0.3 0.3 0.3>
    componentLinearVelScale=<0.5 0.5 0.5>
    gravityOverride {z=-100}
    physicsBodyDefinitions [{
      boundBone {boneName=dyn_hair_14}
      boxExtents=<6 6 8>
      constraintSetup {angularLimitsMax=<5 5 15>, angularLimitsMin=<-5 -5 -15>, angularTarget=<0 0 0>, linearAxesMin=<-0.1 0 0>}
      localJointOffset=<1 0 0>
     }]
   }
   n24 = Anim Dynamics - Bone: dyn_hair_15!(ComponentPose=@n23, bAlphaBoolEnabled:bool=True, ExternalForce="0,0,0") -> (Pose) {
    alphaInputType=Bool
    angularBiasOverride=0.9
    angularDampingOverride=0.99
    angularSpringConstant=2000
    bAngularSpring=true
    bOverrideAngularDamping=true
    bUseGravityOverride=true
    alphaBoolBlend {blendInTime=0.1, blendOption=Custom, blendOutTime=0.2}
    boundBone {boneName=dyn_hair_15}
    componentAppliedLinearAccClamp=<2500 2500 2500>
    componentLinearAccScale=<0.3 0.3 0.3>
    componentLinearVelScale=<0.5 0.5 0.5>
    gravityOverride {z=-100}
    physicsBodyDefinitions [{
      boundBone {boneName=dyn_hair_15}
      boxExtents=<5 5 10>
      constraintSetup {angularLimitsMax=<5 5 15>, angularLimitsMin=<-5 -5 -15>, angularTarget=<0 0 0>, linearAxesMin=<-0.1 0 0>}
      localJointOffset=<1 0 0>
     }]
   }
   n25 = Anim Dynamics - Bone: dyn_hair_16!(ComponentPose=@n24, bAlphaBoolEnabled:bool=True, ExternalForce="0,0,0") -> (Pose) {
    alphaInputType=Bool
    angularBiasOverride=0.9
    angularDampingOverride=0.99
    angularSpringConstant=2000
    bAngularSpring=true
    bOverrideAngularDamping=true
    bUseGravityOverride=true
    alphaBoolBlend {blendInTime=0.1, blendOption=Custom, blendOutTime=0.2}
    boundBone {boneName=dyn_hair_16}
    componentAppliedLinearAccClamp=<500 500 500>
    componentLinearAccScale=<0.3 0.3 0.3>
    componentLinearVelScale=<0.5 0.5 0.5>
    gravityOverride {z=-100}
    physicsBodyDefinitions [{
      boundBone {boneName=dyn_hair_16}
      boxExtents=<5 5 10>
      constraintSetup {angularLimitsMax=<5 5 15>, angularLimitsMin=<-5 -5 -15>, angularTarget=<0 0 0>, linearAxesMin=<-0.1 0 0>}
      localJointOffset=<1 0 0>
     }]
   }
   n26 = Anim Dynamics - Bone: dyn_hair_17!(ComponentPose=@n25, bAlphaBoolEnabled:bool=True, ExternalForce="0,0,0") -> (Pose) {
    alphaInputType=Bool
    angularBiasOverride=0.9
    angularDampingOverride=0.99
    angularSpringConstant=2000
    bAngularSpring=true
    bOverrideAngularDamping=true
    bUseGravityOverride=true
    alphaBoolBlend {blendInTime=0.1, blendOption=Custom, blendOutTime=0.2}
    boundBone {boneName=dyn_hair_17}
    componentAppliedLinearAccClamp=<1000 1000 1000>
    componentLinearAccScale=<0.3 0.3 0.3>
    componentLinearVelScale=<0.5 0.5 0.5>
    gravityOverride {z=-100}
    physicsBodyDefinitions [{
      boundBone {boneName=dyn_hair_17}
      boxExtents=<6 6 8>
      constraintSetup {angularLimitsMax=<5 5 15>, angularLimitsMin=<-5 -5 -15>, angularTarget=<0 0 0>, linearAxesMin=<-0.1 0 0>}
      localJointOffset=<1 0 0>
     }]
   }
   n27 = Anim Dynamics - Bone: dyn_hair_18!(ComponentPose=@n26, bAlphaBoolEnabled:bool=True, ExternalForce="0,0,0") -> (Pose) {
    alphaInputType=Bool
    angularBiasOverride=0.9
    angularDampingOverride=0.99
    angularSpringConstant=2000
    bAngularSpring=true
    bOverrideAngularDamping=true
    bUseGravityOverride=true
    alphaBoolBlend {blendInTime=0.1, blendOption=Custom, blendOutTime=0.2}
    boundBone {boneName=dyn_hair_18}
    componentAppliedLinearAccClamp=<1500 1500 1500>
    componentLinearAccScale=<0.3 0.3 0.3>
    componentLinearVelScale=<0.5 0.5 0.5>
    gravityOverride {z=-100}
    physicsBodyDefinitions [{
      boundBone {boneName=dyn_hair_18}
      boxExtents=<5 5 10>
      constraintSetup {angularLimitsMax=<5 5 15>, angularLimitsMin=<-5 -5 -15>, angularTarget=<0 0 0>, linearAxesMin=<-0.1 0 0>}
      localJointOffset=<1 0 0>
     }]
   }
   n28 = Anim Dynamics - Bone: dyn_hair_19!(ComponentPose=@n27, bAlphaBoolEnabled:bool=True, ExternalForce="0,0,0") -> (Pose) {
    alphaInputType=Bool
    angularBiasOverride=0.9
    angularDampingOverride=0.99
    angularSpringConstant=2000
    bAngularSpring=true
    bOverrideAngularDamping=true
    bUseGravityOverride=true
    alphaBoolBlend {blendInTime=0.1, blendOption=Custom, blendOutTime=0.2}
    boundBone {boneName=dyn_hair_19}
    componentAppliedLinearAccClamp=<2000 2000 2000>
    componentLinearAccScale=<0.3 0.3 0.3>
    componentLinearVelScale=<0.5 0.5 0.5>
    gravityOverride {z=-100}
    physicsBodyDefinitions [{
      boundBone {boneName=dyn_hair_19}
      boxExtents=<6 6 8>
      constraintSetup {angularLimitsMax=<5 5 15>, angularLimitsMin=<-5 -5 -15>, angularTarget=<0 0 0>, linearAxesMin=<-0.1 0 0>}
      localJointOffset=<1 0 0>
     }]
   }
   n29 = Anim Dynamics - Bone: dyn_hair_20!(ComponentPose=@n28, bAlphaBoolEnabled:bool=True, ExternalForce="0,0,0") -> (Pose) {
    alphaInputType=Bool
    angularBiasOverride=0.9
    angularDampingOverride=0.99
    angularSpringConstant=2000
    bAngularSpring=true
    bOverrideAngularDamping=true
    bUseGravityOverride=true
    alphaBoolBlend {blendInTime=0.1, blendOption=Custom, blendOutTime=0.2}
    boundBone {boneName=dyn_hair_20}
    componentAppliedLinearAccClamp=<2500 2500 2500>
    componentLinearAccScale=<0.3 0.3 0.3>
    componentLinearVelScale=<0.5 0.5 0.5>
    gravityOverride {z=-100}
    physicsBodyDefinitions [{
      boundBone {boneName=dyn_hair_20}
      boxExtents=<5 5 10>
      constraintSetup {angularLimitsMax=<5 5 15>, angularLimitsMin=<-5 -5 -15>, angularTarget=<0 0 0>, linearAxesMin=<-0.1 0 0>}
      localJointOffset=<1 0 0>
     }]
   }
   n30 = Anim Dynamics - Bone: dyn_hair_21!(ComponentPose=@n29, bAlphaBoolEnabled:bool=True, ExternalForce="0,0,0") -> (Pose) {
    alphaInputType=Bool
    angularBiasOverride=0.9
    angularDampingOverride=0.99
    angularSpringConstant=2000
    bAngularSpring=true
    bOverrideAngularDamping=true
    bUseGravityOverride=true
    alphaBoolBlend {blendInTime=0.1, blendOption=Custom, blendOutTime=0.2}
    boundBone {boneName=dyn_hair_21}
    componentAppliedLinearAccClamp=<500 500 500>
    componentLinearAccScale=<0.3 0.3 0.3>
    componentLinearVelScale=<0.5 0.5 0.5>
    gravityOverride {z=-100}
    physicsBodyDefinitions [{
      boundBone {boneName=dyn_hair_21}
      boxExtents=<5 5 10>
      constraintSetup {angularLimitsMax=<5 5 15>, angularLimitsMin=<-5 -5 -15>, angularTarget=<0 0 0>, linearAxesMin=<-0.1 0 0>}
      localJointOffset=<1 0 0>
     }]
   }
   n31 = Anim Dynamics - Bone: dyn_hair_22!(ComponentPose=@n30, bAlphaBoolEnabled:bool=True, ExternalForce="0,0,0") -> (Pose) {
    alphaInputType=Bool
    angularBiasOverride=0.9
    angularDampingOverride=0.99
    angularSpringConstant=2000
    bAngularSpring=true
    bOverrideAngularDamping=true
    bUseGravityOverride=true
    alphaBoolBlend {blendInTime=0.1, blendOption=Custom, blendOutTime=0.2}
    boundBone {boneName=dyn_hair_22}
    componentAppliedLinearAccClamp=<1000 1000 1000>
    componentLinearAccScale=<0.3 0.3 0.3>
    componentLinearVelScale=<0.6 0.6 0.6>
    gravityOverride {z=10}
    physicsBodyDefinitions [{
      boundBone {boneName=dyn_hair_22}
      boxExtents=<6 6 8>
      constraintSetup {angularLimitsMax=<10 5 15>, angularLimitsMin=<-10 -5 -15>, angularTarget=<0 0 0>, linearAxesMin=<-0.1 0 0>}
      localJointOffset=<1 0 0>
     }]
   }
   n32 = Anim Dynamics - Bone: dyn_hair_23!(ComponentPose=@n31, bAlphaBoolEnabled:bool=True, ExternalForce="0,0,0") -> (Pose) {
    alphaInputType=Bool
    angularBiasOverride=0.9
    angularDampingOverride=0.99
    angularSpringConstant=2000
    bAngularSpring=true
    bOverrideAngularDamping=true
    bUseGravityOverride=true
    alphaBoolBlend {blendInTime=0.1, blendOption=Custom, blendOutTime=0.2}
    boundBone {boneName=dyn_hair_23}
    componentAppliedLinearAccClamp=<1500 1500 1500>
    componentLinearAccScale=<0.3 0.3 0.3>
    componentLinearVelScale=<0.5 0.5 0.5>
    gravityOverride {z=-100}
    physicsBodyDefinitions [{
      boundBone {boneName=dyn_hair_23}
      boxExtents=<5 5 10>
      constraintSetup {angularLimitsMax=<5 5 15>, angularLimitsMin=<-5 -5 -15>, angularTarget=<0 0 0>, linearAxesMin=<-0.1 0 0>}
      localJointOffset=<1 0 0>
     }]
   }
   n33 = Anim Dynamics - Bone: dyn_hair_24!(ComponentPose=@n32, bAlphaBoolEnabled:bool=True, ExternalForce="0,0,0") -> (Pose) {
    alphaInputType=Bool
    angularBiasOverride=0.9
    angularDampingOverride=0.99
    angularSpringConstant=2000
    bAngularSpring=true
    bOverrideAngularDamping=true
    bUseGravityOverride=true
    alphaBoolBlend {blendInTime=0.1, blendOption=Custom, blendOutTime=0.2}
    boundBone {boneName=dyn_hair_24}
    componentAppliedLinearAccClamp=<2000 2000 2000>
    componentLinearAccScale=<0.3 0.3 0.3>
    componentLinearVelScale=<0.5 0.5 0.5>
    gravityOverride {z=-100}
    physicsBodyDefinitions [{
      boundBone {boneName=dyn_hair_24}
      boxExtents=<6 6 8>
      constraintSetup {angularLimitsMax=<5 5 15>, angularLimitsMin=<-5 -5 -15>, angularTarget=<0 0 0>, linearAxesMin=<-0.1 0 0>}
      localJointOffset=<1 0 0>
     }]
   }
   n34 = Anim Dynamics - Bone: dyn_hair_25!(ComponentPose=@n33, bAlphaBoolEnabled:bool=True, ExternalForce="0,0,0") -> (Pose) {
    alphaInputType=Bool
    angularBiasOverride=0.9
    angularDampingOverride=0.99
    angularSpringConstant=2000
    bAngularSpring=true
    bOverrideAngularDamping=true
    bUseGravityOverride=true
    alphaBoolBlend {blendInTime=0.1, blendOption=Custom, blendOutTime=0.2}
    boundBone {boneName=dyn_hair_25}
    componentAppliedLinearAccClamp=<2500 2500 2500>
    componentLinearAccScale=<0.3 0.3 0.3>
    componentLinearVelScale=<0.5 0.5 0.5>
    gravityOverride {z=-100}
    physicsBodyDefinitions [{
      boundBone {boneName=dyn_hair_25}
      boxExtents=<5 5 10>
      constraintSetup {angularLimitsMax=<5 5 15>, angularLimitsMin=<-5 -5 -15>, angularTarget=<0 0 0>, linearAxesMin=<-0.1 0 0>}
      localJointOffset=<1 0 0>
     }]
   }
   n35 = Anim Dynamics - Bone: dyn_hair_26!(ComponentPose=@n34, bAlphaBoolEnabled:bool=True, ExternalForce="0,0,0") -> (Pose) {
    alphaInputType=Bool
    angularBiasOverride=0.9
    angularDampingOverride=0.99
    angularSpringConstant=2000
    bAngularSpring=true
    bOverrideAngularDamping=true
    bUseGravityOverride=true
    alphaBoolBlend {blendInTime=0.1, blendOption=Custom, blendOutTime=0.2}
    boundBone {boneName=dyn_hair_26}
    componentAppliedLinearAccClamp=<500 500 500>
    componentLinearAccScale=<0.3 0.3 0.3>
    componentLinearVelScale=<0.5 0.5 0.5>
    gravityOverride {z=-100}
    physicsBodyDefinitions [{
      boundBone {boneName=dyn_hair_26}
      boxExtents=<5 5 10>
      constraintSetup {angularLimitsMax=<5 5 15>, angularLimitsMin=<-5 -5 -15>, angularTarget=<0 0 0>, linearAxesMin=<-0.1 0 0>}
      localJointOffset=<1 0 0>
     }]
   }
   n36 = Anim Dynamics - Bone: dyn_hair_27!(ComponentPose=@n35, bAlphaBoolEnabled:bool=True, ExternalForce="0,0,0") -> (Pose) {
    alphaInputType=Bool
    angularBiasOverride=0.9
    angularDampingOverride=0.99
    angularSpringConstant=2000
    bAngularSpring=true
    bOverrideAngularDamping=true
    bUseGravityOverride=true
    alphaBoolBlend {blendInTime=0.1, blendOption=Custom, blendOutTime=0.2}
    boundBone {boneName=dyn_hair_27}
    componentAppliedLinearAccClamp=<1000 1000 1000>
    componentLinearAccScale=<0.3 0.3 0.3>
    componentLinearVelScale=<0.5 0.5 0.5>
    gravityOverride {z=-100}
    physicsBodyDefinitions [{
      boundBone {boneName=dyn_hair_27}
      boxExtents=<6 6 8>
      constraintSetup {angularLimitsMax=<5 5 15>, angularLimitsMin=<-5 -5 -15>, angularTarget=<0 0 0>, linearAxesMin=<-0.1 0 0>}
      localJointOffset=<1 0 0>
     }]
   }
   n37 = Anim Dynamics - Bone: dyn_hair_28!(ComponentPose=@n36, bAlphaBoolEnabled:bool=True, ExternalForce="0,0,0") -> (Pose) {
    alphaInputType=Bool
    angularBiasOverride=0.9
    angularDampingOverride=0.99
    angularSpringConstant=2000
    bAngularSpring=true
    bOverrideAngularDamping=true
    bUseGravityOverride=true
    alphaBoolBlend {blendInTime=0.1, blendOption=Custom, blendOutTime=0.2}
    boundBone {boneName=dyn_hair_28}
    componentAppliedLinearAccClamp=<1500 1500 1500>
    componentLinearAccScale=<0.3 0.3 0.3>
    componentLinearVelScale=<0.5 0.5 0.5>
    gravityOverride {z=-100}
    physicsBodyDefinitions [{
      boundBone {boneName=dyn_hair_28}
      boxExtents=<5 5 10>
      constraintSetup {angularLimitsMax=<5 5 15>, angularLimitsMin=<-5 -5 -15>, angularTarget=<0 0 0>, linearAxesMin=<-0.1 0 0>}
      localJointOffset=<1 0 0>
     }]
   }
   n38 = Anim Dynamics - Bone: dyn_hair_29!(ComponentPose=@n37, bAlphaBoolEnabled:bool=True, ExternalForce="0,0,0") -> (Pose) {
    alphaInputType=Bool
    angularBiasOverride=0.9
    angularDampingOverride=0.99
    angularSpringConstant=2000
    bAngularSpring=true
    bOverrideAngularDamping=true
    bUseGravityOverride=true
    alphaBoolBlend {blendInTime=0.1, blendOption=Custom, blendOutTime=0.2}
    boundBone {boneName=dyn_hair_29}
    componentAppliedLinearAccClamp=<2000 2000 2000>
    componentLinearAccScale=<0.3 0.3 0.3>
    componentLinearVelScale=<0.5 0.5 0.5>
    gravityOverride {z=-100}
    physicsBodyDefinitions [{
      boundBone {boneName=dyn_hair_29}
      boxExtents=<6 6 8>
      constraintSetup {angularLimitsMax=<5 5 15>, angularLimitsMin=<-5 -5 -15>, angularTarget=<0 0 0>, linearAxesMin=<-0.1 0 0>}
      localJointOffset=<1 0 0>
     }]
   }
   n39 = Anim Dynamics - Bone: dyn_hair_30!(ComponentPose=@n38, bAlphaBoolEnabled:bool=True, ExternalForce="0,0,0") -> (Pose) {
    alphaInputType=Bool
    angularBiasOverride=0.9
    angularDampingOverride=0.99
    angularSpringConstant=2000
    bAngularSpring=true
    bOverrideAngularDamping=true
    bUseGravityOverride=true
    alphaBoolBlend {blendInTime=0.1, blendOption=Custom, blendOutTime=0.2}
    boundBone {boneName=dyn_hair_30}
    componentAppliedLinearAccClamp=<2500 2500 2500>
    componentLinearAccScale=<0.3 0.3 0.3>
    componentLinearVelScale=<0.5 0.5 0.5>
    gravityOverride {z=-100}
    physicsBodyDefinitions [{
      boundBone {boneName=dyn_hair_30}
      boxExtents=<5 5 10>
      constraintSetup {angularLimitsMax=<5 5 15>, angularLimitsMin=<-5 -5 -15>, angularTarget=<0 0 0>, linearAxesMin=<-0.1 0 0>}
      localJointOffset=<1 0 0>
     }]
   }
   Output Pose(Result=Component To Local(ComponentPose=@n39)) {layerGroup=None}
   n40 = "Transform (Modify) Bone - Bone: dyn_headphones_MSTR"!(
    ComponentPose=Local To Component(LocalPose=Copy Pose From Mesh() {bUseAttachedParent=true})
    bAlphaBoolEnabled:bool=True
    Translation="1,0,0"
    Rotation="0,5,0"
    Scale="1,1,1"
   ) -> (Pose) {
    alphaInputType=Bool
    rotationMode=BMM_Additive
    rotationSpace=BCS_BoneSpace
    translationMode=BMM_Additive
    translationSpace=BCS_BoneSpace
    alphaBoolBlend {blendInTime=0.2, blendOption=Linear, blendOutTime=0.4}
    boneToModify {boneName=dyn_headphones_MSTR}
    rotation {yaw=5}
    translation {x=1}
   }
   n41 = "Constraint - dyn_headphones (RR)"!(ComponentPose=@n40, Alpha=1, ConstraintWeights_0=0.3, ConstraintWeights_1:float=0.7) -> (Pose) {
    alphaBoolBlend {blendOption=Linear}
    boneToModify {boneName=dyn_headphones}
    constraintSetup [{transformType=Rotation, targetBone {boneName=head}}, {transformType=Rotation, targetBone {boneName=spine_05}}]
    constraintWeights [1,1]
   }
   n42 = Look At - Bone: dyn_headphones_MSTR(ComponentPose=@n41, Alpha=1) {
    alphaBoolBlend {blendOption=Linear}
    boneToModify {boneName=dyn_headphones_MSTR}
    lookAtLocation=<6.8482 36.3494 0>
    lookAtTarget {boneReference {boneName=neck_01}}
   }
   n43 = "Transform (Modify) Bone - Bone: dyn_headphones"(
    ComponentPose=@n42
    bAlphaBoolEnabled=Get Heaphone_crouch() -> (Heaphone_crouch:bool)
    Translation="-0.2,0,0"
    Rotation="0,0,0"
    Scale="1,1,1"
   ) {
    alphaInputType=Bool
    rotationMode=BMM_Additive
    rotationSpace=BCS_BoneSpace
    translationMode=BMM_Additive
    translationSpace=BCS_BoneSpace
    alphaBoolBlend {blendInTime=0.2, blendOption=Linear, blendOutTime=0.4}
    boneToModify {boneName=dyn_headphones}
    translation {x=-0.2}
   }
   n44 = "Transform (Modify) Bone - Bone: dyn_headphones"(ComponentPose=@n43, bAlphaBoolEnabled=Get bIsTargeting(), Translation="-0.6,0,0", Rotation="0,0,0", Scale="1,1,1") {
    alphaInputType=Bool
    rotationMode=BMM_Additive
    rotationSpace=BCS_BoneSpace
    translationMode=BMM_Additive
    translationSpace=BCS_BoneSpace
    alphaBoolBlend {blendInTime=0.2, blendOption=Linear, blendOutTime=0.4}
    boneToModify {boneName=dyn_headphones}
    translation {x=-0.6}
   }
   n45 = "Transform (Modify) Bone - Bone: dyn_headphones_r_MSTR"(ComponentPose=@n44, Alpha=Get Headphone_Right(), Translation="0,0,0", Rotation="0,0,15", Scale="1,1,1") {
    rotationMode=BMM_Additive
    rotationSpace=BCS_BoneSpace
    translationMode=BMM_Additive
    translationSpace=BCS_BoneSpace
    alphaBoolBlend {blendInTime=0.2, blendOption=Linear, blendOutTime=0.6}
    boneToModify {boneName=dyn_headphones_r_MSTR}
    rotation {roll=15}
   }
   n46 = "Transform (Modify) Bone - Bone: dyn_headphones_l_MSTR"(ComponentPose=@n45, Alpha=Get Left_ear() -> (Left_ear:double), Translation="0,0,0", Rotation="0,0,8", Scale="1,1,1") {
    rotationMode=BMM_Additive
    rotationSpace=BCS_BoneSpace
    translationMode=BMM_Additive
    translationSpace=BCS_BoneSpace
    alphaBoolBlend {blendInTime=0.2, blendOption=Linear, blendOutTime=0.6}
    boneToModify {boneName=dyn_headphones_l_MSTR}
    rotation {roll=8}
   }
   n47 = "Transform (Modify) Bone - Bone: dyn_headphones_r_MSTR"!(
    ComponentPose=@n46
    Alpha=1
    Translation="0,0,0"
    Rotation_Roll:float=Get Head rotation X full() -> (Head rotation X full:float)
    Rotation_Pitch:float=0
    Rotation_Yaw:float=0
    Scale="1,1,1"
   ) -> (Pose) {
    rotationMode=BMM_Additive
    rotationSpace=BCS_BoneSpace
    translationMode=BMM_Additive
    translationSpace=BCS_BoneSpace
    alphaBoolBlend {blendInTime=0.2, blendOption=Linear, blendOutTime=0.6}
    boneToModify {boneName=dyn_headphones_r_MSTR}
   }
   n48 = Get bIsPlayingMeleeAnim()
   n49 = "Constraint - dyn_headphones (RR)"!(ComponentPose=@n47, bAlphaBoolEnabled:bool=@n48, ConstraintWeights_0=0.3, ConstraintWeights_1=0.7) -> (Pose) {
    alphaInputType=Bool
    alphaBoolBlend {blendInTime=0.2, blendOption=Linear, blendOutTime=0.8}
    boneToModify {boneName=dyn_headphones}
    constraintSetup [{transformType=Rotation, targetBone {boneName=spine_05}}, {transformType=Rotation, targetBone {boneName=head}}]
    constraintWeights [1,1]
   }
   n50 = "Transform (Modify) Bone - Bone: dyn_headphones"(ComponentPose=@n49, bAlphaBoolEnabled=@n48, Translation="-1,0,0", Rotation="0,0,0", Scale="1,1,1") {
    alphaInputType=Bool
    rotationMode=BMM_Additive
    rotationSpace=BCS_BoneSpace
    translationMode=BMM_Additive
    translationSpace=BCS_BoneSpace
    alphaBoolBlend {blendInTime=0.2, blendOption=Linear, blendOutTime=0.8}
    boneToModify {boneName=dyn_headphones}
    translation {x=-1}
   }
   n51 = "Constraint - dyn_face_acc (PPPPPPPP)"!(
    ComponentPose=@n50
    Alpha=1
    ConstraintWeights_0=0.5006
    ConstraintWeights_1:float=0.0339
    ConstraintWeights_2:float=0.0318
    ConstraintWeights_3:float=0.0044
    ConstraintWeights_4:float=0.0054
    ConstraintWeights_5:float=0.1703
    ConstraintWeights_6:float=0.0032
    ConstraintWeights_7:float=0.2504
   ) -> (Pose) {
    alphaBoolBlend {blendOption=Linear}
    boneToModify {boneName=dyn_face_acc}
    constraintSetup [
     {targetBone {boneName=FACIAL_C_LipLower}}
     {targetBone {boneName=FACIAL_L_LipLower}}
     {targetBone {boneName=FACIAL_R_LipLower}}
     {targetBone {boneName=FACIAL_L_LipLowerOuter}}
     {targetBone {boneName=FACIAL_R_LipLowerOuter}}
     {targetBone {boneName=FACIAL_C_Jaw}}
     {targetBone {boneName=FACIAL_L_ChinSide}}
     {targetBone {boneName=FACIAL_C_Chin}}
    ]
    constraintWeights [1,1,1,1,1,1,1,1]
   }
   Save cached pose 'Headphones'(Pose=Component To Local(ComponentPose=@n51)) {cacheName=Headphones}
   n52 = "Transform (Modify) Bone - Bone: dyn_headphones_l"!(
    ComponentPose
    bAlphaBoolEnabled=True
    Translation_X:double=0
    Translation_Y:double=0
    Translation_Z:double=Get Left_ear()
    Rotation="0,0,0"
    Scale="1,1,1"
   ) -> (Pose) {
    alphaInputType=Bool
    rotationMode=BMM_Additive
    rotationSpace=BCS_BoneSpace
    translationMode=BMM_Additive
    translationSpace=BCS_BoneSpace
    alphaBoolBlend {blendInTime=0.8, blendOption=ExpIn, blendOutTime=0.8}
    boneToModify {boneName=dyn_headphones_l}
    translation {z=-1}
   }
   n53 = "Transform (Modify) Bone - Bone: dyn_headphones_r"!(
    ComponentPose=@n52
    bAlphaBoolEnabled=True
    Translation_X:double=Get Headphone_Right()
    Translation_Y:double=0
    Translation_Z:double=0
    Rotation="0,0,0"
    Scale="1,1,1"
   ) -> (Pose) {
    alphaInputType=Bool
    rotationMode=BMM_Additive
    rotationSpace=BCS_BoneSpace
    translationMode=BMM_Additive
    translationSpace=BCS_BoneSpace
    alphaBoolBlend {blendInTime=0.8, blendOption=ExpIn, blendOutTime=0.8}
    boneToModify {boneName=dyn_headphones_r}
    translation {z=1}
   }
   n54 = "Transform (Modify) Bone - Bone: dyn_headphones"!(
    ComponentPose=@n53
    bAlphaBoolEnabled=NOT Boolean(A=Get bIsPlayingMeleeAnim())
    Translation_X:double=Get Head rotation hedphones UP() -> (Head rotation hedphones UP:float)
    Translation_Y:double=Get Head rotation hedphones Back() -> (Head rotation hedphones Back:double)
    Translation_Z:double=0
    Rotation_Roll:float=Get Hand rise() -> (Hand rise:float)
    Rotation_Pitch:float=0
    Rotation_Yaw:float=Get Head rotation() -> (Head rotation:float)
    Scale="1,1,1"
   ) -> (Pose) {
    alphaInputType=Bool
    rotationMode=BMM_Additive
    rotationSpace=BCS_BoneSpace
    translationMode=BMM_Additive
    translationSpace=BCS_BoneSpace
    alphaBoolBlend {blendInTime=0.2, blendOption=Linear, blendOutTime=0.6}
    boneToModify {boneName=dyn_headphones}
    translation {x=-1.7}
   }
   Reroute Node(InputPin:Component Space Pose Link Structure=@n54) -> (OutputPin:Component Space Pose Link Structure)
  ]}
 EventGraph {nodes [
   n1 = Event BlueprintUpdateAnimation()
   n2 = GetAlphaFromReferenceTransformDelta(
    SkeletalMeshComponent=GetOwnerSkeletalMeshComponent()
    BoneName=upperarm_r
    BaseName=clavicle_r
    InputRange=(X=0,Y=0)
    OutputRange=(X=0,Y=6)
    TwistAxis="1,0,0"
   )
   > Set Hand rise(Hand rise:float=float * float(A=@n2.RawOutput, B=-0.2) {is=title K2Node_PromotableOperator}) -> (Output_Get:float)
   n3 = float * float(A=@n2.RawOutput, B=-0.05)
   > Set Headphone_left(Headphone_left:double=FInterpTo(
     Current=Get Headphone_left() -> (Headphone_left:double)
     Target="Clamp (Float)"(Value=@n3, Min=-1, Max=0)
     DeltaTime=@n1.DeltaTimeX
     InterpSpeed=7
    )) -> (Output_Get:double)
   n4 = GetAlphaFromReferenceTransformDelta(
    SkeletalMeshComponent=GetOwnerSkeletalMeshComponent()
    BoneName=head
    BaseName=spine_05
    DeltaComponentType=EulerX
    InputRange=(X=0,Y=0)
    TwistAxis="1,0,0"
   )
   > Set Head rotation X(Head rotation X:float=FInterpTo(
     Current=Get Head rotation X() -> (Head rotation X:float)
     Target="Clamp (Float)"(Value=@n4.RawOutput, Min=0, Max=1.7)
     DeltaTime=@n1.DeltaTimeX
     InterpSpeed=1
    )) -> (Output_Get:float)
   > Set Head rotation X full(Head rotation X full:float=FInterpTo(Current=Get Head rotation X full(), Target="Clamp (Float)"(Value=@n4.RawOutput, Min=-15, Max=0), DeltaTime=@n1.DeltaTimeX, InterpSpeed=1)) -> (Output_Get:float)
   > Set Left_ear(Left_ear:double=float + float(A=Get Headphone_left(), B=Get Head rotation X())) -> (Output_Get:double)
   n5 = float * float(A=@n2.RawOutput, B=0.02)
   > Set Headphone_Right(Headphone_Right:float=FInterpTo(Current=Get Headphone_Right(), Target="Clamp (Float)"(Value=@n5, Min=0, Max=1), DeltaTime=@n1.DeltaTimeX, InterpSpeed=3)) -> (Output_Get:float)
   n6 = Get bIsMoving2D()
   n7 = Get bIsTargeting()
   > Set Heaphone_crouch(Heaphone_crouch:bool=AND Boolean!(A=Get bIsCrouching(), B=NOT Boolean(A=@n6), C:bool=NOT Boolean(A=@n7)) -> (ReturnValue)) -> (Output_Get:bool)
   n8 = GetOwnerSkeletalMeshComponent()
   Set Head rotation(Head rotation:float=float * float(
     A=GetAlphaFromReferenceTransformDelta(SkeletalMeshComponent=@n8, BoneName=head, BaseName=spine_05, InputRange=(X=0,Y=0), TwistAxis="1,0,0").RawOutput
     B=-0.5
    )) -> (Output_Get:float)
   n9 = float * float(
    A=GetAlphaFromReferenceTransformDelta(SkeletalMeshComponent=GetOwnerSkeletalMeshComponent(), BoneName=head, BaseName=neck_01, InputRange=(X=0,Y=0), TwistAxis="1,0,0").RawOutput
    B=0.1
   )
   > Set Head rotation hedphones UP(Head rotation hedphones UP:float="Clamp (Float)"(Value=@n9, Min=-1, Max=0)) -> (Output_Get:float)
   > Set Head rotation hedphones Back(Head rotation hedphones Back:double="Clamp (Float)"(Value=@n9, Min=-50, Max=0)) -> (Output_Get:double)
  ]}
]
variables [
 BodyMesh {type=SkeletalMeshComponent}
 Hand rise {type=number}
 Head rotation {type=number}
 Head rotation hedphones Back {type=number}
 Head rotation hedphones UP {type=number}
 Head rotation X {type=number}
 Head rotation X full {type=number}
 Headphone_left {type=number}
 Headphone_Right {type=number}
 Heaphone_crouch {type=boolean}
 Left_ear {type=number}
]
